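# Default Megascans (MS) preset master material from Bridge, a Default Lit surface with swappable texture map parameters.
# Samples Albedo, Metalness (red channel), Normal and a packed DR map whose green channel supplies roughness.
# All four feed MF_ObjAdjustments, which holds the adjustment logic, and the material works through material attributes.
# Albedo, Metalness and Normal use shared world-group wrap samplers; DR does not.
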Material /Bridge/MSPresets/M_MS_Default_Material/M_MS_Default_Material
graph {
 properties {bUseMaterialAttributes=true, pixelDepthOffsetMode=PDOM_Legacy, usedShadingModels=Default Lit}
 nodes [
  n1 = TextureSampleParameter2D() {
   group=07 - Texture Maps
   parameterName=Albedo
   samplerSource=SSM_Wrap_WorldGroupSettings
   sortPriority=0
   texture=/Engine/EngineMaterials/DefaultDiffuse
  }
  n2 = TextureSampleParameter2D() {
   group=07 - Texture Maps
   parameterName=Metalness
   samplerSource=SSM_Wrap_WorldGroupSettings
   samplerType=SAMPLERTYPE_LinearColor
   sortPriority=1
   texture=/Bridge/MSPresets/MSTextures/BlackPlaceholder
  }
  n3 = TextureSampleParameter2D() {
   group=07 - Texture Maps
   parameterName=Normal
   samplerSource=SSM_Wrap_WorldGroupSettings
   samplerType=SAMPLERTYPE_Normal
   sortPriority=3
   texture=/Bridge/MSPresets/MSTextures/Placeholder_Normal
  }
  MaterialFunctionCall!(
   Albedo=@n1.RGB
   Metallic=@n2.R
   Roughness=TextureSampleParameter2D() {desc=DR, parameterName=DR, samplerType=SAMPLERTYPE_LinearColor, texture=/Bridge/MSPresets/MSTextures/WhitePlaceholder}.G
   Tangent Normal=@n3.RGB
  ) {materialFunction=/Bridge/MSPresets/M_MS_Default_Material/Functions/MF_ObjAdjustments}
 ]
}
